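# Looping morning outdoor ambience cue for the Asteria Industrial Town POI, using the Athena ambient attenuation and sound class.
# A Quality Level root selects between one quad branch and three identical stereo branches.
# The quad branch randomly picks one of five looping beds, passing it through slight pitch modulation and two nested looping envelopers.
# Each stereo branch plays the same third-variation loop through a looping pitch/volume envelope and random modulation.
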SoundCue /Game/Sounds/Asteria/POI/IndustrialTown/Ambient_IndustrialTown_Morning_Loop_Cue
root=QualityLevel_0
nodes [
 {
  id=Enveloper_0
  children [Enveloper_3]
  properties {
   bLoop=true
   bLoopIndefinitely=true
   loopEnd=100
   pitchCurve {editorCurveData {keys [<at 0 0.8>]}}
   volumeCurve {editorCurveData {keys [<at 0 1 RCIM_Cubic>,<at 50 0.9 RCIM_Cubic>,<at 100 1 RCIM_Cubic>]}}
  }
 }
 {
  id=Enveloper_1
  children [WavePlayer_6]
  properties {
   bLoop=true
   bLoopIndefinitely=true
   loopEnd=30
   pitchCurve {editorCurveData {keys [{interpMode=RCIM_Cubic, time=0}, <at 15.9468 0.9181 RCIM_Cubic>, {interpMode=RCIM_Cubic, time=30}]}}
   volumeCurve {editorCurveData {keys [<at 0 1 RCIM_Cubic>,<at 16.1421 0.921 RCIM_Cubic>,<at 30 1 RCIM_Cubic>]}}
  }
 }
 {
  id=Enveloper_2
  children [WavePlayer_7]
  properties {
   bLoop=true
   bLoopIndefinitely=true
   loopEnd=30
   pitchCurve {editorCurveData {keys [{interpMode=RCIM_Cubic, time=0}, <at 15.9468 0.9181 RCIM_Cubic>, {interpMode=RCIM_Cubic, time=30}]}}
   volumeCurve {editorCurveData {keys [<at 0 1 RCIM_Cubic>,<at 16.1421 0.921 RCIM_Cubic>,<at 30 1 RCIM_Cubic>]}}
  }
 }
 {
  id=Enveloper_3
  children [Modulator_11]
  properties {
   bLoop=true
   bLoopIndefinitely=true
   loopEnd=30
   pitchCurve {editorCurveData {keys [{interpMode=RCIM_Cubic, time=0}, <at 15.9468 0.9181 RCIM_Cubic>, {interpMode=RCIM_Cubic, time=30}]}}
   volumeCurve {editorCurveData {keys [<at 0 1>]}}
  }
 }
 {
  id=Enveloper_4
  children [WavePlayer_8]
  properties {
   bLoop=true
   bLoopIndefinitely=true
   loopEnd=30
   pitchCurve {editorCurveData {keys [{interpMode=RCIM_Cubic, time=0}, <at 15.9468 0.9181 RCIM_Cubic>, {interpMode=RCIM_Cubic, time=30}]}}
   volumeCurve {editorCurveData {keys [<at 0 1 RCIM_Cubic>,<at 16.1421 0.921 RCIM_Cubic>,<at 30 1 RCIM_Cubic>]}}
  }
 }
 {id=Modulator_11, children [Random_1], properties {pitchMax=1, pitchMin=0.97}}
 {id=Modulator_2, children [Enveloper_1], properties {pitchMax=1.1, pitchMin=0.9, volumeMax=1.1, volumeMin=0.9}}
 {id=Modulator_3, children [Enveloper_2], properties {pitchMax=1.1, pitchMin=0.9, volumeMax=1.1, volumeMin=0.9}}
 {id=Modulator_4, children [Enveloper_4], properties {pitchMax=1.1, pitchMin=0.9, volumeMax=1.1, volumeMin=0.9}}
 {id=QualityLevel_0, children [Enveloper_0,Modulator_2,Modulator_3,Modulator_4]}
 {id=Random_1, children [WavePlayer_1,WavePlayer_2,WavePlayer_3,WavePlayer_4,WavePlayer_5], properties {weights [1,1,1,1,1]}}
 {id=WavePlayer_1, properties {bLooping=true, soundWaveAssetPtr=&/Ambience_Industrial_Morning_01_Quad}}
 {id=WavePlayer_2, properties {bLooping=true, soundWaveAssetPtr=&/Ambience_Industrial_Morning_02_Quad}}
 {id=WavePlayer_3, properties {bLooping=true, soundWaveAssetPtr=&/Ambience_Industrial_Morning_03_Quad}}
 {id=WavePlayer_4, properties {bLooping=true, soundWaveAssetPtr=&/Ambience_Industrial_Morning_04_Quad}}
 {id=WavePlayer_5, properties {bLooping=true, soundWaveAssetPtr=&/Ambience_Industrial_Morning_05_Quad}}
 {id=WavePlayer_6, properties {bLooping=true, soundWaveAssetPtr=&/Ambience_Industrial_Morning_03_Stereo}}
 {id=WavePlayer_7, properties {bLooping=true, soundWaveAssetPtr=&Ambience_Industrial_Morning_03_Stereo}}
 {id=WavePlayer_8, properties {bLooping=true, soundWaveAssetPtr=&Ambience_Industrial_Morning_03_Stereo}}
]
properties {
 attenuationSettings=&Fort_Ambient_Athena_Att
 maxDistance=2097152
 soundClassObject=&Fort_Ambient_Outdoors_Athena
 volumeMultiplier=0.551
}
