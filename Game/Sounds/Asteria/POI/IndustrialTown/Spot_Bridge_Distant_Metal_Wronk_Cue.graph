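# Ambient SoundCue for the Industrial Town bridge spot that plays intermittent distant metal "wronk" sounds.
# An endless loop waits a random delay, then picks one of six equally weighted waves reused from Hollow Base cave spots, with modulated pitch.
# An in-graph Attenuation node overrides falloff on top of the cue's distant-creak attenuation asset.
# The quality switch feeds only its first two levels, leaving the other two empty.
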SoundCue /Game/Sounds/Asteria/POI/IndustrialTown/Spot_Bridge_Distant_Metal_Wronk_Cue
root=QualityLevel_0
nodes [
 {
  id=Attenuation_0
  children [Modulator_0]
  properties {bOverrideAttenuation=true, attenuationOverrides {falloffDistance=7000, attenuationShapeExtents {x=2000}}}
 }
 {id=Delay_0, children [Attenuation_0], properties {delayMax=16, delayMin=3}}
 {id=Looping_0, children [Delay_0]}
 {id=Modulator_0, children [Random_0], properties {pitchMin=0.9}}
 {id=QualityLevel_0, children [Looping_0,Looping_0,null,null]}
 {id=Random_0, children [WavePlayer_0,WavePlayer_1,WavePlayer_2,WavePlayer_4,WavePlayer_6,WavePlayer_7], properties {weights [1,1,1,1,1,1]}}
 {id=WavePlayer_0, properties {soundWaveAssetPtr=/Game/Sounds/Artemis/POI/Hollow_Base/Spots/Spot_Cave_Distant_Metal_Wronk_1}}
 {id=WavePlayer_1, properties {soundWaveAssetPtr=/Game/Sounds/Artemis/POI/Hollow_Base/Spots/Spot_Cave_Distant_Metal_Wronk_2}}
 {id=WavePlayer_2, properties {soundWaveAssetPtr=/Game/Sounds/Artemis/POI/Hollow_Base/Spots/Spot_Cave_Distant_Metal_Wronk_3}}
 {id=WavePlayer_4, properties {soundWaveAssetPtr=/Game/Sounds/Artemis/POI/Hollow_Base/Spots/Spot_Cave_Distant_Metal_Wronk_5}}
 {id=WavePlayer_6, properties {soundWaveAssetPtr=/Game/Sounds/Artemis/POI/Hollow_Base/Spots/Spot_Cave_Distant_Metal_Wronk_7}}
 {id=WavePlayer_7, properties {soundWaveAssetPtr=/Game/Sounds/Artemis/POI/Hollow_Base/Spots/Spot_Cave_Distant_Metal_Wronk_8}}
]
properties {
 attenuationSettings=/Game/Sounds/Attenuation/Fort_TileAmbient_OneShot_Distant_Creak_Att
 bEnableBusSends=false
 bEnableSubmixSends=false
 maxDistance=9000
 soundClassObject=&Fort_Ambient_Outdoors_Athena
 volumeMultiplier=0.576
 attenuationOverrides {falloffDistance=6000, stereoSpread=10000, attenuationShapeExtents {x=2000}}
}
